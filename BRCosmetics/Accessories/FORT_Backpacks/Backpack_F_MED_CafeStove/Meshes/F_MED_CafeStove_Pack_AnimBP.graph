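# Animation Blueprint for the CafeStove backpack cosmetic, adding procedural secondary motion to its straps and dangling trinket.
# After a small additive offset on the pack bone, chained AnimDynamics nodes simulate the upper strap, lower strap and trinket bones with angular springs.
# Planar and spherical limits tied to the pack bone constrain some links.
# Each update refreshes a shared gravity override through a deprecated helper and lowers velocity and acceleration scales while skydiving.
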AnimBlueprint /BRCosmetics/Accessories/FORT_Backpacks/Backpack_F_MED_CafeStove/Meshes/F_MED_CafeStove_Pack_AnimBP
target_skeleton=&/F_MED_CafeStove_Pack_Skeleton
graphs [
 AnimGraph {nodes [
   n1 = "Transform (Modify) Bone - Bone: pack"(ComponentPose=Local To Component(LocalPose=Input Pose()), Translation="0,-1,0", Rotation="0,0,0") {
    translationMode=BMM_Additive
    translationSpace=BCS_BoneSpace
    alphaBoolBlend {blendOption=Linear}
    boneToModify {boneName=pack}
    translation {y=-1}
   }
   n2 = Get Gravity Override()
   n3 = Get Component Linear Acc Scale() -> (Component Linear Acc Scale:Vector)
   n4 = Get Component Linear Vel Scale() -> (Component Linear Vel Scale:Vector)
   n5 = Anim Dynamics - Bone: dyn_strap_upper_01!(
    ComponentPose=@n1
    Alpha=1
    GravityOverride:Vector=@n2
    ComponentLinearAccScale:Vector=@n3
    ComponentLinearVelScale:Vector=@n4
    ExternalForce="0,0,0"
   ) -> (Pose) {
    angularDampingOverride=0.999
    angularSpringConstant=300
    bAngularSpring=true
    bOverrideAngularDamping=true
    bUseGravityOverride=true
    simulationSpace=RootRelative
    alphaBoolBlend {blendOption=Linear}
    boundBone {boneName=dyn_strap_upper_01}
    componentAppliedLinearAccClamp=<1000 1000 1000>
    componentLinearAccScale=<0.3 0.3 0.3>
    componentLinearVelScale=<0.3 0.3 0.3>
    gravityOverride {z=-980}
    physicsBodyDefinitions [{
      boundBone {boneName=dyn_strap_upper_01}
      boxExtents=<20 20 20>
      constraintSetup {angularLimitsMax=<0 0 50>, angularLimitsMin=<0 -10 0>, angularTarget=<0 0 0>}
      localJointOffset=<1.5 0 0>
     }]
   }
   n6 = Anim Dynamics - Bone: dyn_strap_upper_02!(
    ComponentPose=@n5
    Alpha=1
    GravityOverride:Vector=@n2
    ComponentLinearAccScale:Vector=@n3
    ComponentLinearVelScale:Vector=@n4
    ExternalForce="0,0,0"
   ) -> (Pose) {
    angularDampingOverride=0.999
    angularSpringConstant=300
    bAngularSpring=true
    bOverrideAngularDamping=true
    bUseGravityOverride=true
    simulationSpace=RootRelative
    alphaBoolBlend {blendOption=Linear}
    boundBone {boneName=dyn_strap_upper_02}
    componentAppliedLinearAccClamp=<1000 1000 1000>
    componentLinearAccScale=<0.3 0.3 0.3>
    componentLinearVelScale=<0.3 0.3 0.3>
    gravityOverride {z=-980}
    physicsBodyDefinitions [{
      boundBone {boneName=dyn_strap_upper_02}
      boxExtents=<14 14 14>
      constraintSetup {angularLimitsMax=<0 10 100>, angularLimitsMin=<0 -15 -10>, angularTarget=<0 0 0>}
      localJointOffset=<1.9 0 0>
     }]
    planarLimits [{drivingBone {boneName=pack}, planeTransform {location=<-13.8754 -0.9243 42.3803>, rotation=<rot 46.192 2.0608 94.7018>}}]
   }
   n7 = Anim Dynamics - Bone: dyn_strap_upper_03!(
    ComponentPose=@n6
    Alpha=1
    GravityOverride:Vector=@n2
    ComponentLinearAccScale:Vector=@n3
    ComponentLinearVelScale:Vector=@n4
    ExternalForce="0,0,0"
   ) -> (Pose) {
    angularDampingOverride=0.999
    angularSpringConstant=300
    bAngularSpring=true
    bOverrideAngularDamping=true
    bUseGravityOverride=true
    simulationSpace=RootRelative
    alphaBoolBlend {blendOption=Linear}
    boundBone {boneName=dyn_strap_upper_03}
    componentAppliedLinearAccClamp=<1000 1000 1000>
    componentLinearAccScale=<0.3 0.3 0.3>
    componentLinearVelScale=<0.3 0.3 0.3>
    gravityOverride {z=-980}
    physicsBodyDefinitions [{
      boundBone {boneName=dyn_strap_upper_03}
      boxExtents=<12 12 12>
      constraintSetup {angularLimitsMax=<0 30 60>, angularLimitsMin=<0 -60 -10>, angularTarget=<0 0 0>}
      localJointOffset=<1 0 0>
     }]
    planarLimits [{drivingBone {boneName=pack}, planeTransform {location=<-16.9361 -0.9628 38.9663>, rotation=<rot 46.1442 4.0614 95.6237>}}]
   }
   n8 = Anim Dynamics - Bone: dyn_strap_lower_01!(
    ComponentPose=@n7
    Alpha=1
    GravityOverride:Vector=@n2
    ComponentLinearAccScale:Vector=@n3
    ComponentLinearVelScale:Vector=@n4
    ExternalForce="0,0,0"
   ) -> (Pose) {
    angularDampingOverride=0.999
    angularSpringConstant=300
    bAngularSpring=true
    bOverrideAngularDamping=true
    bUseGravityOverride=true
    simulationSpace=RootRelative
    alphaBoolBlend {blendOption=Linear}
    boundBone {boneName=dyn_strap_lower_01}
    componentAppliedLinearAccClamp=<1000 1000 1000>
    componentLinearAccScale=<0.3 0.3 0.3>
    componentLinearVelScale=<0.3 0.3 0.3>
    gravityOverride {z=-980}
    physicsBodyDefinitions [{
      boundBone {boneName=dyn_strap_lower_01}
      boxExtents=<20 20 20>
      constraintSetup {angularLimitsMax=<0 0 20>, angularLimitsMin=<0 -10 -10>, angularTarget=<0 0 0>}
      localJointOffset=<1.5 0 0>
     }]
   }
   n9 = Anim Dynamics - Bone: dyn_strap_lower_02!(
    ComponentPose=@n8
    Alpha=1
    GravityOverride:Vector=@n2
    ComponentLinearAccScale:Vector=@n3
    ComponentLinearVelScale:Vector=@n4
    ExternalForce="0,0,0"
   ) -> (Pose) {
    angularDampingOverride=0.999
    angularSpringConstant=300
    bAngularSpring=true
    bOverrideAngularDamping=true
    bUseGravityOverride=true
    simulationSpace=RootRelative
    alphaBoolBlend {blendOption=Linear}
    boundBone {boneName=dyn_strap_lower_02}
    componentAppliedLinearAccClamp=<1000 1000 1000>
    componentLinearAccScale=<0.3 0.3 0.3>
    componentLinearVelScale=<0.3 0.3 0.3>
    gravityOverride {z=-980}
    physicsBodyDefinitions [{
      boundBone {boneName=dyn_strap_lower_02}
      boxExtents=<14 14 14>
      constraintSetup {angularLimitsMax=<0 10 100>, angularLimitsMin=<0 -15 -10>, angularTarget=<0 0 0>}
      localJointOffset=<1.9 0 0>
     }]
   }
   n10 = Anim Dynamics - Bone: dyn_strap_lower_03!(
    ComponentPose=@n9
    Alpha=1
    GravityOverride:Vector=@n2
    ComponentLinearAccScale:Vector=@n3
    ComponentLinearVelScale:Vector=@n4
    ExternalForce="0,0,0"
   ) -> (Pose) {
    angularDampingOverride=0.999
    angularSpringConstant=300
    bAngularSpring=true
    bOverrideAngularDamping=true
    bUseGravityOverride=true
    simulationSpace=RootRelative
    alphaBoolBlend {blendOption=Linear}
    boundBone {boneName=dyn_strap_lower_03}
    componentAppliedLinearAccClamp=<1000 1000 1000>
    componentLinearAccScale=<0.3 0.3 0.3>
    componentLinearVelScale=<0.3 0.3 0.3>
    gravityOverride {z=-980}
    physicsBodyDefinitions [{
      boundBone {boneName=dyn_strap_lower_03}
      boxExtents=<12 12 12>
      constraintSetup {angularLimitsMax=<0 30 110>, angularLimitsMin=<0 -60 -10>, angularTarget=<0 0 0>}
      localJointOffset=<2 0 0>
     }]
   }
   n11 = Anim Dynamics - Bone: dyn_trinket_01!(
    ComponentPose=@n10
    Alpha=1
    GravityOverride:Vector=@n2
    ComponentLinearAccScale:Vector=@n3
    ComponentLinearVelScale:Vector=@n4
    ExternalForce="0,0,0"
   ) -> (Pose) {
    angularDampingOverride=0.9999
    angularSpringConstant=100
    bAngularSpring=true
    bOverrideAngularDamping=true
    bUseGravityOverride=true
    simulationSpace=RootRelative
    alphaBoolBlend {blendOption=Linear}
    boundBone {boneName=dyn_trinket_01}
    componentAppliedLinearAccClamp=<2000 2000 2000>
    componentLinearAccScale=<0.3 0.3 0.3>
    componentLinearVelScale=<0.3 0.3 0.3>
    gravityOverride {z=-980}
    physicsBodyDefinitions [{
      boundBone {boneName=dyn_trinket_01}
      boxExtents=<10 10 10>
      constraintSetup {angularLimitsMax=<10 30 30>, angularLimitsMin=<-10 -30 -60>, angularTarget=<0 0 0>, linearAxesMax=<0 0 0.3>, linearAxesMin=<0 0 -0.3>}
      localJointOffset=<1 0 0>
     }]
   }
   n12 = Anim Dynamics - Bone: dyn_trinket_02!(
    ComponentPose=@n11
    Alpha=1
    GravityOverride:Vector=@n2
    ComponentLinearAccScale:Vector=@n3
    ComponentLinearVelScale:Vector=@n4
    ExternalForce="0,0,0"
   ) -> (Pose) {
    angularDampingOverride=0.9999
    angularSpringConstant=100
    bAngularSpring=true
    bOverrideAngularDamping=true
    bUseGravityOverride=true
    simulationSpace=RootRelative
    alphaBoolBlend {blendOption=Linear}
    boundBone {boneName=dyn_trinket_02}
    componentAppliedLinearAccClamp=<2000 2000 2000>
    componentLinearAccScale=<0.3 0.3 0.3>
    componentLinearVelScale=<0.3 0.3 0.3>
    gravityOverride {z=-980}
    physicsBodyDefinitions [{
      boundBone {boneName=dyn_trinket_02}
      boxExtents=<8 8 8>
      constraintSetup {angularLimitsMax=<10 90 90>, angularLimitsMin=<-10 -90 -90>, angularTarget=<0 0 0>, linearAxesMax=<0 0 0.1>, linearAxesMin=<0 0 -0.1>}
      localJointOffset=<1 0 0>
     }]
   }
   n13 = Anim Dynamics - Bone: dyn_trinket_03!(
    ComponentPose=@n12
    Alpha=1
    GravityOverride:Vector=@n2
    ComponentLinearAccScale:Vector=@n3
    ComponentLinearVelScale:Vector=@n4
    ExternalForce="0,0,0"
   ) -> (Pose) {
    angularDampingOverride=0.9999
    angularSpringConstant=100
    bAngularSpring=true
    bOverrideAngularDamping=true
    bUseGravityOverride=true
    simulationSpace=RootRelative
    alphaBoolBlend {blendOption=Linear}
    boundBone {boneName=dyn_trinket_03}
    componentAppliedLinearAccClamp=<2000 2000 2000>
    componentLinearAccScale=<0.3 0.3 0.3>
    componentLinearVelScale=<0.3 0.3 0.3>
    gravityOverride {z=-980}
    physicsBodyDefinitions [{
      boundBone {boneName=dyn_trinket_03}
      boxExtents=<8 8 8>
      constraintSetup {angularLimitsMax=<10 90 90>, angularLimitsMin=<-10 -90 -90>, angularTarget=<0 0 0>}
      localJointOffset=<1 0 0>
     }]
   }
   n14 = Anim Dynamics - Bone: dyn_trinket_04!(
    ComponentPose=@n13
    Alpha=1
    GravityOverride:Vector=@n2
    ComponentLinearAccScale:Vector=@n3
    ComponentLinearVelScale:Vector=@n4
    ExternalForce="0,0,0"
   ) -> (Pose) {
    angularDampingOverride=0.9999
    angularSpringConstant=100
    bAngularSpring=true
    bOverrideAngularDamping=true
    bUseGravityOverride=true
    bUseSphericalLimits=true
    simulationSpace=RootRelative
    alphaBoolBlend {blendOption=Linear}
    boundBone {boneName=dyn_trinket_04}
    componentAppliedLinearAccClamp=<2000 2000 2000>
    componentLinearAccScale=<0.3 0.3 0.3>
    componentLinearVelScale=<0.3 0.3 0.3>
    gravityOverride {z=-980}
    physicsBodyDefinitions [{
      boundBone {boneName=dyn_trinket_04}
      boxExtents=<8 8 8>
      constraintSetup {angularLimitsMax=<10 90 90>, angularLimitsMin=<-10 -90 -90>, angularTarget=<0 0 0>}
      localJointOffset=<2 0 0>
     }]
    planarLimits [{drivingBone {boneName=pack}, planeTransform {location=<-8.3646 -0.0843 -15.4791>, rotation=<rot 20.1651 179.93 177.8121>}}]
    sphericalLimits [{limitRadius=2.9359, drivingBone {boneName=pack}, sphereLocalOffset=<-13.4205 0.2593 -13.6509>}]
   }
   Output Pose(Result=Component To Local(ComponentPose=@n14)) {layerGroup=None}
  ]}
 EventGraph {nodes [
   Event BlueprintUpdateAnimation()
   > n1 = Gravity Override State Function DEPRECATED(
    CharacterPartAnimInstance=Self-Reference()
    GameTimeInSeconds=Get Game Time in Seconds_0() -> (Game Time in Seconds_0:double)
    GravityOverrideParams=Get Gravity Override Params_0() -> (Gravity Override Params_0:Gravity Override Params Struct Structure)
   )
   > Set Gravity Override(Gravity Override=@n1)
   > n2 = Branch(Condition=Get bIsSkyDiving())
   > Set Component Linear Vel Scale(Component Linear Vel Scale:Vector="0.1,0.1,0.1") -> (Output_Get:Vector)
   > Set Component Linear Acc Scale(Component Linear Acc Scale:Vector="0.1,0.1,0.1") -> (Output_Get:Vector)
   Set Component Linear Vel Scale(execute=@n2.else, Component Linear Vel Scale="0.3,0.3,0.3")
   > Set Component Linear Acc Scale(Component Linear Acc Scale="0.3,0.3,0.3")
   TryGetPawnOwner()
  ]}
]
variables [
 Component Linear Acc Scale {type=Vector, default=<0.3 0.3 0.3>}
 Component Linear Vel Scale {type=Vector, default=<0.3 0.3 0.3>}
 Game Time in Seconds_0 {type=number}
 Gravity Override {type=Vector}
 Gravity Override Params_0 {type=GravityOverrideParamsStruct}
]
